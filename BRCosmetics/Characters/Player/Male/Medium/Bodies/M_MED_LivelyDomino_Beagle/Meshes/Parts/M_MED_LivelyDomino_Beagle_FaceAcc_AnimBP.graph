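# Post-body animation Blueprint for the LivelyDomino Beagle face accessory mesh, adding secondary physics motion to its ear and hair bones.
# It copies the attached parent's pose, then runs a serial chain of single-bone, spring-driven Anim Dynamics nodes.
# Ears use head-driven spherical limits; hair uses varied alphas and mostly low gravity, with its root simulated relative to spine_05.
# The update event is empty and the declared variables are unused.
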AnimBlueprint /BRCosmetics/Characters/Player/Male/Medium/Bodies/M_MED_LivelyDomino_Beagle/Meshes/Parts/M_MED_LivelyDomino_Beagle_FaceAcc_AnimBP
target_skeleton=&/M_MED_LivelyDomino_Beagle_FaceAcc_Skeleton
graphs [
 AnimGraph {nodes [
   n1 = Anim Dynamics - Bone: dyn_ear_l_01(ComponentPose=Local To Component(LocalPose=Copy Pose From Mesh() {bUseAttachedParent=true}), Alpha=1, ExternalForce="0,0,0") {
    angularDampingOverride=0.99
    angularSpringConstant=75
    bAngularSpring=true
    bGravityOverrideInSimSpace=true
    bOverrideAngularDamping=true
    bUseSphericalLimits=true
    simulationSpace=RootRelative
    alphaBoolBlend {blendOption=Linear}
    boundBone {boneName=dyn_ear_l_01}
    componentLinearAccScale=<1 2 1>
    componentLinearVelScale=<1 2 1>
    physicsBodyDefinitions [{
      sphereCollisionRadius=0
      boundBone {boneName=dyn_ear_l_01}
      boxExtents=<10 20 20>
      constraintSetup {angularLimitsMax=<15 15 5>, angularLimitsMin=<-15 -15 -5>, angularTarget=<1 0 0>}
      localJointOffset=<1 0 0.2>
     }]
    sphericalLimits [{limitRadius=6.5, drivingBone {boneName=head}, sphereLocalOffset=<4 0 0>}]
   }
   n2 = Anim Dynamics - Bone: dyn_ear_l_02(ComponentPose=@n1, Alpha=1, ExternalForce="0,0,0") {
    angularDampingOverride=0.99
    angularSpringConstant=75
    bAngularSpring=true
    bGravityOverrideInSimSpace=true
    bOverrideAngularDamping=true
    bUseSphericalLimits=true
    simulationSpace=RootRelative
    alphaBoolBlend {blendOption=Linear}
    boundBone {boneName=dyn_ear_l_02}
    componentLinearAccScale=<1 2 1>
    componentLinearVelScale=<1 2 1>
    physicsBodyDefinitions [{
      sphereCollisionRadius=0
      boundBone {boneName=dyn_ear_l_02}
      boxExtents=<10 20 20>
      constraintSetup {angularLimitsMax=<15 15 5>, angularLimitsMin=<-15 -15 -5>, angularTarget=<1 0 0>}
      localJointOffset=<1 0 0.2>
     }]
    sphericalLimits [{limitRadius=6.5, drivingBone {boneName=head}, sphereLocalOffset=<4 0 0>}]
   }
   n3 = Anim Dynamics - Bone: dyn_ear_r_01(ComponentPose=@n2, Alpha=1, ExternalForce="0,0,0") {
    angularDampingOverride=0.99
    angularSpringConstant=85
    bAngularSpring=true
    bGravityOverrideInSimSpace=true
    bOverrideAngularDamping=true
    bUseSphericalLimits=true
    simulationSpace=RootRelative
    alphaBoolBlend {blendOption=Linear}
    boundBone {boneName=dyn_ear_r_01}
    componentLinearAccScale=<1 2 1>
    componentLinearVelScale=<1 2 1>
    physicsBodyDefinitions [{
      sphereCollisionRadius=0
      boundBone {boneName=dyn_ear_r_01}
      boxExtents=<13 25 25>
      constraintSetup {angularLimitsMax=<15 15 5>, angularLimitsMin=<-15 -15 -5>, angularTarget=<1 0 0>}
      localJointOffset=<1 0 -0.2>
     }]
    sphericalLimits [{limitRadius=6.5, drivingBone {boneName=head}, sphereLocalOffset=<4 0 0>}]
   }
   n4 = Anim Dynamics - Bone: dyn_ear_r_02(ComponentPose=@n3, Alpha=1, ExternalForce="0,0,0") {
    angularDampingOverride=0.99
    angularSpringConstant=85
    bAngularSpring=true
    bGravityOverrideInSimSpace=true
    bOverrideAngularDamping=true
    bUseSphericalLimits=true
    simulationSpace=RootRelative
    alphaBoolBlend {blendOption=Linear}
    boundBone {boneName=dyn_ear_r_02}
    componentLinearAccScale=<1 2 1>
    componentLinearVelScale=<1 2 1>
    physicsBodyDefinitions [{
      sphereCollisionRadius=0
      boundBone {boneName=dyn_ear_r_02}
      boxExtents=<13 25 25>
      constraintSetup {angularLimitsMax=<15 15 5>, angularLimitsMin=<-15 -15 -5>, angularTarget=<1 0 0>}
      localJointOffset=<1 0 -0.2>
     }]
    sphericalLimits [{limitRadius=6.5, drivingBone {boneName=head}, sphereLocalOffset=<4 0 0>}]
   }
   n5 = Anim Dynamics - Bone: dyn_hair_mid_root(ComponentPose=@n4, Alpha=0.5, ExternalForce="0,0,0") {
    angularDampingOverride=-0.01
    angularSpringConstant=75
    bAngularSpring=true
    bGravityOverrideInSimSpace=true
    bOverrideAngularDamping=true
    simulationSpace=BoneRelative
    alphaBoolBlend {blendOption=Linear}
    boundBone {boneName=dyn_hair_mid_root}
    componentAppliedLinearAccClamp=<800 800 800>
    componentLinearAccScale=<0.1 0.2 0.1>
    componentLinearVelScale=<0.1 0.2 0.1>
    physicsBodyDefinitions [{
      sphereCollisionRadius=0
      boundBone {boneName=dyn_hair_mid_root}
      boxExtents=<30 60 60>
      constraintSetup {angularLimitsMax=<1 1 1>, angularLimitsMin=<-1 -1 -1>, angularTarget=<1 0 0>}
      localJointOffset=<5 0 0>
     }]
    relativeSpaceBone {boneName=spine_05}
   }
   n6 = Anim Dynamics - Bone: dyn_hair_mid_01(ComponentPose=@n5, Alpha=0.75, ExternalForce="0,0,0") {
    angularDampingOverride=0.99
    angularSpringConstant=75
    bAngularSpring=true
    bGravityOverrideInSimSpace=true
    bOverrideAngularDamping=true
    gravityScale=0.2
    simulationSpace=RootRelative
    alphaBoolBlend {blendOption=Linear}
    boundBone {boneName=dyn_hair_mid_01}
    componentLinearAccScale=<1 2 1>
    componentLinearVelScale=<1 2 1>
    physicsBodyDefinitions [{
      sphereCollisionRadius=0
      boundBone {boneName=dyn_hair_mid_01}
      boxExtents=<20 40 40>
      constraintSetup {angularLimitsMax=<5 5 5>, angularLimitsMin=<-5 -5 -35>, angularTarget=<1 0 0>}
      localJointOffset=<6 -1 0.2>
     }]
   }
   n7 = Anim Dynamics - Bone: dyn_hair_mid_02(ComponentPose=@n6, Alpha=1, ExternalForce="0,0,0") {
    angularDampingOverride=0.99
    angularSpringConstant=75
    bAngularSpring=true
    bGravityOverrideInSimSpace=true
    bOverrideAngularDamping=true
    gravityScale=0.2
    simulationSpace=RootRelative
    alphaBoolBlend {blendOption=Linear}
    boundBone {boneName=dyn_hair_mid_02}
    componentLinearAccScale=<1 2 1>
    componentLinearVelScale=<1 2 1>
    physicsBodyDefinitions [{
      sphereCollisionRadius=0
      boundBone {boneName=dyn_hair_mid_02}
      boxExtents=<10 20 20>
      constraintSetup {angularLimitsMax=<25 25 5>, angularLimitsMin=<-25 -25 -35>, angularTarget=<1 0 0>}
      localJointOffset=<6 -1 0.2>
     }]
   }
   n8 = Anim Dynamics - Bone: dyn_hair_mid_03(ComponentPose=@n7, Alpha=0.85, ExternalForce="0,0,0") {
    angularDampingOverride=0.99
    angularSpringConstant=70
    bAngularSpring=true
    bGravityOverrideInSimSpace=true
    bOverrideAngularDamping=true
    gravityScale=0.2
    simulationSpace=RootRelative
    alphaBoolBlend {blendOption=Linear}
    boundBone {boneName=dyn_hair_mid_03}
    componentLinearAccScale=<1 2 1>
    componentLinearVelScale=<1 2 1>
    physicsBodyDefinitions [{
      sphereCollisionRadius=0
      boundBone {boneName=dyn_hair_mid_03}
      boxExtents=<20 40 40>
      constraintSetup {angularLimitsMax=<15 15 15>, angularLimitsMin=<-15 -15 -35>, angularTarget=<1 0 0>}
      localJointOffset=<6 -1 0.2>
     }]
   }
   n9 = Anim Dynamics - Bone: dyn_hair_mid_04(ComponentPose=@n8, Alpha=1, ExternalForce="0,0,0") {
    angularDampingOverride=0.99
    angularSpringConstant=78
    bAngularSpring=true
    bGravityOverrideInSimSpace=true
    bOverrideAngularDamping=true
    gravityScale=0.2
    simulationSpace=RootRelative
    alphaBoolBlend {blendOption=Linear}
    boundBone {boneName=dyn_hair_mid_04}
    componentLinearAccScale=<1 2 1>
    componentLinearVelScale=<1 2 1>
    physicsBodyDefinitions [{
      sphereCollisionRadius=0
      boundBone {boneName=dyn_hair_mid_04}
      boxExtents=<10 20 20>
      constraintSetup {angularLimitsMax=<25 25 5>, angularLimitsMin=<-25 -25 -35>, angularTarget=<1 0 0>}
      localJointOffset=<6 -1 0.2>
     }]
   }
   n10 = Anim Dynamics - Bone: dyn_hair_l_01(ComponentPose=@n9, Alpha=0.65, ExternalForce="0,0,0") {
    angularDampingOverride=0.99
    angularSpringConstant=75
    bAngularSpring=true
    bGravityOverrideInSimSpace=true
    bOverrideAngularDamping=true
    gravityScale=0.2
    simulationSpace=RootRelative
    alphaBoolBlend {blendOption=Linear}
    boundBone {boneName=dyn_hair_l_01}
    componentLinearAccScale=<1 2 1>
    componentLinearVelScale=<1 2 1>
    physicsBodyDefinitions [{
      sphereCollisionRadius=0
      boundBone {boneName=dyn_hair_l_01}
      boxExtents=<10 20 20>
      constraintSetup {angularLimitsMax=<15 15 5>, angularLimitsMin=<-15 -15 -15>, angularTarget=<1 0 0>}
      localJointOffset=<6 -1 0.2>
     }]
   }
   n11 = Anim Dynamics - Bone: dyn_hair_r_01(ComponentPose=@n10, Alpha=0.8, ExternalForce="0,0,0") {
    angularDampingOverride=0.9999
    angularSpringConstant=80
    bAngularSpring=true
    bGravityOverrideInSimSpace=true
    bOverrideAngularDamping=true
    gravityScale=0.2
    simulationSpace=RootRelative
    alphaBoolBlend {blendOption=Linear}
    boundBone {boneName=dyn_hair_r_01}
    componentLinearAccScale=<1 2 1>
    componentLinearVelScale=<1 2 1>
    physicsBodyDefinitions [{
      sphereCollisionRadius=0
      boundBone {boneName=dyn_hair_r_01}
      boxExtents=<10 20 20>
      constraintSetup {angularLimitsMax=<15 15 5>, angularLimitsMin=<-15 -15 -15>, angularTarget=<1 0 0>}
      localJointOffset=<6 -1 0.2>
     }]
   }
   n12 = Anim Dynamics - Bone: dyn_hair_mid_05(ComponentPose=@n11, Alpha=1, ExternalForce="0,0,0") {
    angularDampingOverride=0.99
    angularSpringConstant=85
    bAngularSpring=true
    bGravityOverrideInSimSpace=true
    bOverrideAngularDamping=true
    gravityScale=0.2
    simulationSpace=RootRelative
    alphaBoolBlend {blendOption=Linear}
    boundBone {boneName=dyn_hair_mid_05}
    componentLinearAccScale=<1 2 1>
    componentLinearVelScale=<1 2 1>
    physicsBodyDefinitions [{
      sphereCollisionRadius=0
      boundBone {boneName=dyn_hair_mid_05}
      boxExtents=<20 40 40>
      constraintSetup {angularLimitsMax=<5 5 5>, angularLimitsMin=<-5 -5 -35>, angularTarget=<1 0 0>}
      localJointOffset=<6 -1 0.2>
     }]
   }
   n13 = Anim Dynamics - Bone: dyn_hair_mid_06(ComponentPose=@n12, Alpha=0.8, ExternalForce="0,0,0") {
    angularDampingOverride=0.99
    angularSpringConstant=55
    bAngularSpring=true
    bGravityOverrideInSimSpace=true
    bOverrideAngularDamping=true
    gravityScale=0.2
    simulationSpace=RootRelative
    alphaBoolBlend {blendOption=Linear}
    boundBone {boneName=dyn_hair_mid_06}
    componentLinearAccScale=<1 2 1>
    componentLinearVelScale=<1 2 1>
    physicsBodyDefinitions [{
      sphereCollisionRadius=0
      boundBone {boneName=dyn_hair_mid_06}
      boxExtents=<10 20 20>
      constraintSetup {angularLimitsMax=<5 5 5>, angularLimitsMin=<-5 -5 -35>, angularTarget=<1 0 0>}
      localJointOffset=<6 -1 0.2>
     }]
   }
   n14 = Anim Dynamics - Bone: dyn_hair_mid_07(ComponentPose=@n13, Alpha=1, ExternalForce="0,0,0") {
    angularDampingOverride=0.99
    angularSpringConstant=60
    bAngularSpring=true
    bGravityOverrideInSimSpace=true
    bOverrideAngularDamping=true
    gravityScale=0.2
    simulationSpace=RootRelative
    alphaBoolBlend {blendOption=Linear}
    boundBone {boneName=dyn_hair_mid_07}
    componentLinearAccScale=<1 2 1>
    componentLinearVelScale=<1 2 1>
    physicsBodyDefinitions [{
      sphereCollisionRadius=0
      boundBone {boneName=dyn_hair_mid_07}
      boxExtents=<20 40 40>
      constraintSetup {angularLimitsMax=<5 5 5>, angularLimitsMin=<-5 -5 -35>, angularTarget=<1 0 0>}
      localJointOffset=<6 -1 0.2>
     }]
   }
   Output Pose(Result=Component To Local(ComponentPose=@n14)) {layerGroup=None}
  ]}
 EventGraph {nodes [Event BlueprintUpdateAnimation()]}
]
variables [
 CinePosing {type=boolean}
 hat_on/off {type=number}
 hat_on/off_0 {type=boolean}
 head rot {type=Rotator}
 HeadLimit {type=boolean, default=true}
 Is_hemlet {type=boolean}
 X percentage {type=number}
 Y percentage {type=number}
 Z percentage {type=number, default=0.8}
]
